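# Translucent UI material for the lock-on cursor of the Badger Bangs prototype weapon.
# The Reticle icon is sampled through center-scaled UVs, with one branch multiplied by CenterOpacity.
# A vertically panned gradient is boosted and clamped, and a time-driven rotator acts on scaled UVs.
# These chains show no connection to material outputs, and the rotator result, CenterSize and Rotation parameters feed nothing shown.
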Material /Game/Athena/Items/Weapons/Prototype/Badger_Bangs/M_Athena_BadgerBangs_LockOnCursor
graph {
 properties {blendMode=BLEND_Translucent, materialDomain=MD_UI, pixelDepthOffsetMode=PDOM_Legacy, usedShadingModels=Default Lit}
 nodes [
  n1 = TextureSampleParameter2D(UVs=MaterialFunctionCall(UVs=TextureCoordinate(), Texture Scale=Subtract() {constA=0.3, constB=0.11}) {materialFunction=&ScaleUVsByCenter}.UVs) {parameterName=Reticle, samplerSource=SSM_Clamp_WorldGroupSettings, texture=&/BadgerBangsIcon}
  Add(B=@n1.G)
  n2 = TextureSampleParameter2D(UVs=Panner(Coordinate=TextureCoordinate()) {speedY=0.25}) {parameterName=Panner, samplerType=SAMPLERTYPE_LinearColor, texture=/Game/Global/Textures/Default/Masks_and_Gradients/T_VertGradient}
  Clamp(None=Multiply(A=@n2.RGB) {constB=1.25}) {maxDefault=2, minDefault=0.8}
  n3 = MaterialFunctionCall(UVs=TextureCoordinate()) {materialFunction=&ScaleUVsByCenter}
  MaterialFunctionCall(UVs=@n3.UVs, "Rotation Angle (0-1)"=Multiply(A=Time()) {constB=0.6}) {materialFunction=&CustomRotator}
  n4 = TextureSampleParameter2D(UVs=@n3.UVs) {parameterName=Reticle, samplerSource=SSM_Clamp_WorldGroupSettings, texture=&BadgerBangsIcon}
  n5 = Multiply(A=@n4.RGB, B=ScalarParameter() {defaultValue=1, parameterName=CenterOpacity})
  Multiply(A=@n5)
  Multiply(A=@n5) {constB=100}
  ScalarParameter() {defaultValue=0.55, parameterName=CenterSize}
  ScalarParameter() {parameterName=Rotation}
  VectorParameter() {defaultValue=<rgba 1 1 1 0>}
 ]
}
