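# Lock-on cursor widget for the Badger Bangs weapon that marks the current Target.
# A looping timer started on Construct repositions it from the target pawn's spine_05 socket, or at viewport center when Target is invalid.
# A one-shot Scale animation pops size and color, then hands off to an endlessly looping Flash color pulse.
# Construct plays the lock-on sound; Destruct clears the timer and plays the lock-off sound.
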WidgetBlueprint /Figment_BadgerBangs/Gameplay/Athena_BadgerBangs_Figment_LockOnWidget
animations [
 Scale {
  binding_count=2
  length_seconds=0.35
  object=&self:Scale
  sequence {bindings [
    Image_83 {
     id=#1
     possessed_class=Image
     tracks [{
       type=MovieSceneColorTrack
       display_name=Color and Opacity
       sections [{
         type=MovieSceneColorSection
         end_frame=7
         start_frame=0
         channels [
          Color.R {type=MovieSceneScriptingFloatChannel, keys [[0,0.0631],[2,1],[7,0.0499]]}
          Color.G {type=MovieSceneScriptingFloatChannel, keys [[0,0.3466],[2,1],[7,0.4002]]}
          Color.B {type=MovieSceneScriptingFloatChannel, const [1,3]}
          Color.A {type=MovieSceneScriptingFloatChannel, const [1,3]}
         ]
        }]
      }]
    }
    CanvasPanelSlot_1 {
     id=#2
     parent_id=#1
     possessed_class=CanvasPanelSlot
     tracks [{
       type=MovieSceneMarginTrack
       display_name="Offsets (Layout Data)"
       sections [{
         type=MovieSceneMarginSection
         end_frame=7
         start_frame=0
         channels [
          Left {type=MovieSceneScriptingFloatChannel}
          Top {type=MovieSceneScriptingFloatChannel}
          Right {type=MovieSceneScriptingFloatChannel, keys [[0,90],[2,130],[7,50]]}
          Bottom {type=MovieSceneScriptingFloatChannel, keys [[0,90],[2,130],[7,50]]}
         ]
        }]
      }]
    }
   ]}
 }
 Flash {
  binding_count=1
  length_seconds=0.7
  object=&self:Flash
  sequence {bindings [Image_83 {
     id=#1
     possessed_class=Image
     tracks [{
       type=MovieSceneColorTrack
       display_name=Color and Opacity
       sections [{
         type=MovieSceneColorSection
         end_frame=14
         start_frame=0
         channels [
          Color.R {type=MovieSceneScriptingFloatChannel, keys [[0,0.0631],[7,0.2],[14,0.0499]]}
          Color.G {type=MovieSceneScriptingFloatChannel, keys [[0,0.3466],[7,0.4533],[14,0.4002]]}
          Color.B {type=MovieSceneScriptingFloatChannel, const [1,3]}
          Color.A {type=MovieSceneScriptingFloatChannel, const [1,3]}
         ]
        }]
      }]
    }]}
 }
]
graphs [EventGraph {nodes [
   n1 = MoveUI() {is=title K2Node_CustomEvent}
   > n2 = Is Valid(InputObject=Get Target() -> (Target:Actor Object Reference))
   n3 = GetViewportSize!() -> (ReturnValue:Vector 2D Structure)
   SetPositionInViewport!(
    execute=@n2.Is Not Valid
    self:User Widget Object Reference
    Position:Vector 2D Structure=vector2d * float!(
     A:Vector 2D Structure=vector2d - vector2d!(A:Vector 2D Structure=@n3, B:Vector 2D Structure) -> (ReturnValue:Vector 2D Structure)
     B:double=0.5
    ) -> (ReturnValue:Vector 2D Structure)
    bRemoveDPIScale:bool=true
   ) -> (then)
   n4 = Cast To FortPlayerPawn(execute=@n2.Is Valid, Object=Get Target())
   n5 = BroadcastPlayerImpactAtLocation!(
    Player:Player Controller Object Reference=GetOwningPlayer()
    Position:"Vector (by ref)"=GetSocketLocation!(self:Scene Component Object Reference=Get Mesh(self=@n4), InSocketName:Name=spine_05) -> (ReturnValue:Vector)
    bRelativePosition:bool=false
   ) -> (Loudness:float, Duration:float, ReturnValue:bool)
   > SetPositionInViewport(Position=MakeVector2D(X=@n5, Y=@n5.Duration), bRemoveDPIScale=true)
   Event Construct()
   > n6 = Set Timer by Event(Delegate=@n1, Time=0.05, bLooping=true)
   > PlayAnimation(InAnimation=Get Scale() -> (Scale:Widget Animation Object Reference))
   > PlaySound2D(Sound=Get LockinOnSound() -> (LockinOnSound:Sound Base Object Reference))
   Event Destruct()
   > Clear and Invalidate Timer by Handle(Handle=@n6)
   > PlaySound2D(Sound=Get LockinOffSound() -> (LockinOffSound:Sound Base Object Reference))
   "Animation Finished (Scale)"!() -> (OutputDelegate:Delegate, then:Exec) {is=title K2Node_WidgetAnimationEvent}
   > PlayAnimation(InAnimation=Get Flash() -> (Flash:Widget Animation Object Reference), NumLoopsToPlay=0)
   Event PreConstruct()
  ]}]
variables [
 LockinOffSound {type=SoundBase, default=/Game/Sounds/Fort_Weapons/BlackMonday/Bangs/BlackMonday_Bang_TargetOff_Cue}
 LockinOnSound {type=SoundBase, default=/Game/Sounds/Fort_Weapons/BlackMonday/Bangs/BlackMonday_Bang_TargetOn_Cue}
 Target {type=Actor, default=null}
]
widgets {
 info {inheritedWidgetCount=0, parentClass=&UserWidget, rootWidgetClass=&CanvasPanel, widgetCount=2}
 widgets [
  CanvasPanel_29 {
   widgetClassPath=&CanvasPanel
   widget=CanvasPanel_29
   properties {renderTransform {angle=0, scale=<1 1>, shear=<0 0>, translation=<-6 -4>}}
  }
  Image_83 {
   widgetClassPath=&Image
   bIsVariable=true
   parent=CanvasPanel_29
   widget=Image_83
   properties {
    brush {
     bIsDynamicallyLoaded=false
     drawAs=Image
     imageType=NoImage
     mirroring=NoMirror
     resourceName=None
     resourceObject=/Game/Athena/Items/Weapons/Prototype/Badger_Bangs/M_Athena_BadgerBangs_LockOnCursor
     tiling=NoTile
     imageSize=<32 32>
     margin=<ltrb 0 0 0 0>
     outlineSettings {
      bUseBrushTransparency=false
      roundingType=HalfHeightRadius
      width=0
      color {colorUseRule=UseColor_Specified, specifiedColor=<rgba 0 0 0 0>}
      cornerRadii=<0 0 0 0>
     }
     tintColor {colorUseRule=UseColor_Specified, specifiedColor=<rgba 1 1 1 1>}
     uVRegion {bIsValid=false, max=<0 0>, min=<0 0>}
    }
    colorAndOpacity=<rgba 0.302 0.4602 1 1>
    renderTransform {angle=360, scale=<1 1>, shear=<0 0>, translation=<-0.7508 0>}
   }
   slot {
    class=CanvasPanelSlot
    properties {layoutData {alignment=<0.5 0.5>, anchors {maximum=<0.5 0.5>, minimum=<0.5 0.5>}, offsets=<ltrb 5 0 50 50>}}
   }
  }
 ]
}
